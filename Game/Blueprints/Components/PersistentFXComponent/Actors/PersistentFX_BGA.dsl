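; PersistentFX_BGA is an actor Blueprint that hosts persistent FX, with an editor-only component initialized by interface message in the construction script.
; BeginPlay destroys the preview mesh, and the PersistentFXComponent's spawn event stores the spawned actor in currentFxActor.
; Its OnRep runs setupFxActorOnServer only on authority, passing the actor's material overrides to a PersistentFX_DefaultMeshActor.
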(event EventBeginPlay
  (bind _fxmeshpreview_dontmoveme (Variables|Default|GetFxMeshPreview_DontMoveMe))
  (|CallParentFunction)
  (Utilities|IsValid _fxmeshpreview_dontmoveme
    (:Is%20Valid
      (Components|DestroyComponent _fxmeshpreview_dontmoveme))))

(event OnFxActorSpawned(PersistentFXComponent_MoveMe)
  (params (actor :type Actor%20Object%20Reference) (comp :type Persistent%20FXComponent%20Object%20Reference))
  (Variables|Default|SetcurrentFxActor actor))

(fn OnRep_currentFxActor
  (bind _currentfxactor (Variables|Default|GetcurrentFxActor))
  (Networking|SwitchHasAuthority
    (:Authority
      (bind _aspersistent_fx_default_mesh_actor (Utilities|Casting|CastToPersistentFX_DefaultMeshActor (Variables|Default|GetcurrentFxActor))
        (:then
          (CallFunction|SetupFxActoronServer (Variables|/Script/Engine.StaticMeshActor|GetStaticMeshComponent _aspersistent_fx_default_mesh_actor) (Variables|Default|GetcurrentFxActor)))
        (:CastFailed
          (Utilities|IsValid _currentfxactor
            (:Is%20Valid
              (CallFunction|SetupFxActoronServer 0 _currentfxactor))))))))

(fn setupFxActorOnServer
  (params (meshComp :type Static%20Mesh%20Component%20Object%20Reference) (FXActor :type Actor%20Object%20Reference))
  (bind _aspersistent_fx_default_mesh_actor (Utilities|Casting|CastToPersistentFX_DefaultMeshActor FXActor))
  (Variables|/Game/Blueprints/Components/PersistentFXComponent/Actors/PersistentFX_DefaultMeshActor.PersistentFX_DefaultMeshActor_C|SetoverrideMats (Variables|ActorFXSettings|GetFxMaterialOverrides) _aspersistent_fx_default_mesh_actor))

(fn ConstructionScript
  (Class|PersistentFXInterface|EditorOnlyInitEditorComp(Message) (Variables|Default|GetPersistentFX_EditorComp) self))
